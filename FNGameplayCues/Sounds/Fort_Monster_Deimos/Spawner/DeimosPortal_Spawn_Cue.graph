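# Sound cue for a Deimos monster portal spawner spawning an enemy, layering close and distant variants on the enemy spawn sound class.
# A root mixer combines two branches, each with its own close or distant attenuation asset, so the blend shifts with listener distance.
# Each branch applies slight pitch modulation to an equally weighted random pick from its own set of waves.
# Global and per-owner concurrency rules are assigned to limit overlapping spawn sounds.
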SoundCue /FNGameplayCues/Sounds/Fort_Monster_Deimos/Spawner/DeimosPortal_Spawn_Cue
root=Mixer_2
nodes [
 {
  id=Attenuation_2
  children [Modulator_3]
  properties {attenuationSettings=/FNGameplayCues/Sounds/Attenuation/Fort_DeimosPortal_Spawn_Distant_Att}
 }
 {
  id=Attenuation_3
  children [Modulator_1]
  properties {attenuationSettings=/FNGameplayCues/Sounds/Attenuation/Fort_DeimosPortal_Spawn_Close_Att}
 }
 {id=Mixer_2, children [Attenuation_3, Attenuation_2], properties {inputVolume [1,1]}}
 {id=Modulator_1, children [Random_7], properties {pitchMax=1.2, pitchMin=0.9, volumeMax=1, volumeMin=1}}
 {id=Modulator_3, children [Random_5], properties {pitchMax=1.2, pitchMin=0.9, volumeMax=1, volumeMin=1}}
 {
  id=Random_5
  children [WavePlayer_22, WavePlayer_23, WavePlayer_24, WavePlayer_25, WavePlayer_26, WavePlayer_27]
  properties {bShouldExcludeFromBranchCulling=true, weights [1,1,1,1,1,1]}
 }
 {
  id=Random_7
  children [WavePlayer_35, WavePlayer_36, WavePlayer_37, WavePlayer_38]
  properties {bShouldExcludeFromBranchCulling=true, weights [1,1,1,1]}
 }
 {id=WavePlayer_22, properties {soundWaveAssetPtr=&/Deimos_Spawner_Spawn_Enemy_Distant_01}}
 {id=WavePlayer_23, properties {soundWaveAssetPtr=&/Deimos_Spawner_Spawn_Enemy_Distant_02}}
 {id=WavePlayer_24, properties {soundWaveAssetPtr=&/Deimos_Spawner_Spawn_Enemy_Distant_03}}
 {id=WavePlayer_25, properties {soundWaveAssetPtr=&/Deimos_Spawner_Spawn_Enemy_Distant_04}}
 {id=WavePlayer_26, properties {soundWaveAssetPtr=&/Deimos_Spawner_Spawn_Enemy_Distant_05}}
 {id=WavePlayer_27, properties {soundWaveAssetPtr=&/Deimos_Spawner_Spawn_Enemy_Distant_06}}
 {id=WavePlayer_35, properties {soundWaveAssetPtr=&/Deimos_Spawner_Spawn_Enemy_Close_01}}
 {id=WavePlayer_36, properties {soundWaveAssetPtr=&/Deimos_Spawner_Spawn_Enemy_Close_02}}
 {id=WavePlayer_37, properties {soundWaveAssetPtr=&/Deimos_Spawner_Spawn_Enemy_Close_03}}
 {id=WavePlayer_38, properties {soundWaveAssetPtr=&/Deimos_Spawner_Spawn_Enemy_Close_04}}
]
properties {
 maxDistance=18000
 priority=0.5
 soundClassObject=/Game/Sounds/FortSoundClassesAndModes/Fort_Enemy_Spawn
 volumeMultiplier=1
 attenuationOverrides {falloffDistance=4000, attenuationShapeExtents {x=1000}}
 concurrencyOverrides {maxCount=2, resolutionRule=StopFarthestThenPreventNew}
 concurrencySet [/Game/Sounds/Concurrency_Rules/Deimos_Spawner_Global_Concurrency, /FNGameplayCues/Sounds/Concurrency_Rules/Deimos_Spawner_Owner_Concurrency]
}
